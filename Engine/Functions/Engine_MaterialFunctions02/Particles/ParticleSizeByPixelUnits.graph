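# Material function that sizes sprite particles in screen pixels rather than world units, outputting World Position Offset.
# It scales pixel size by view-axis distance, FOV and view width, then rebuilds quad corners from UVs along camera axes around the particle position.
# Further outputs add a pivot offset and convert optional normals from view to tangent space.
# An unconnected pixel-snapping branch in clip space is marked as incorrect.
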MaterialFunction /Engine/Functions/Engine_MaterialFunctions02/Particles/ParticleSizeByPixelUnits
graph {
 properties {
  bExposeToLibrary=true
  description=This function replaces the vertex shader for sprite particles with screen space sprites. The particles XY dimension are interpretted as the number of pixels that the particle should span rather than the number of world space units that the particle should occupy.
  libraryCategoriesText [Particles]
 }
 nodes [
  n1 = FunctionInput(Preview=ParticlePositionWS()) {bUsePreviewValueAsDefault=true, inputName=ParticlePosition, sortPriority=4}
  n2 = SceneTexelSize() {desc=1 pixel}
  n3 = Divide(
   A=MaterialFunctionCall(World Position=@n1) {materialFunction=&TransformToClipSpace}.Clip Space XY
   B=@n2
  )
  n4 = CameraPositionWS()
  n5 = Abs(None=DotProduct(
    A=Subtract(A=@n1, B=@n4)
    B=MaterialFunctionCall() {materialFunction=&CameraDirectionVector}.Result
   )) {desc=Distance}
  MaterialExpressionAppendVector_0 = AppendVector(A=Multiply(A=Ceil(None=@n3), B=@n2), B=@n5) {desc=Z depth}
  Custom!(in=@MaterialExpressionAppendVector_0) {
   code="mul(float4(in.xyz,1),View.ClipToTranslatedWorld)"
   desc="Attempt to transform the clip space value back to world space. Currently, this is incorrect"
   inputs [{
     inputName=in
     input {expression=&self:MaterialExpressionAppendVector_0}
    }]
  }
  n6 = Transform(None=Constant3Vector() {constant=<rgba 1 0 0 0>}) {transformSourceType=TRANSFORMSOURCE_Camera}
  n7 = MaterialFunctionCall(Float2=FunctionInput() {bUsePreviewValueAsDefault=true, inputName=ScreenSpace Pivot Offset, inputType=FunctionInput_Vector2, sortPriority=5}) {materialFunction=&BreakOutFloat2Components}
  n8 = ParticleSize()
  n9 = Custom!(None) {code="GetCotanHalfFieldOfView().x", desc="GetCotanHalfFieldOfView() - seems to always return 1,1", outputType=CMOT_Float2}
  n10 = ViewSize()
  n11 = Multiply(A=ComponentMask(None=Multiply(A=@n9, B=@n10) {constA=1}) {r=true}) {constB=0.5}
  n12 = MaterialFunctionCall(Float2=Multiply(
    A=FunctionInput(Preview=@n8) {bUsePreviewValueAsDefault=true, inputName=Size In Pixels, inputType=FunctionInput_Vector2}
    B=Divide(A=@n5, B=@n11) {desc=Size of a pixel in view space}
   )) {materialFunction=&BreakOutFloat2Components}
  n13 = Multiply(A=Multiply(A=@n6, B=@n7.R), B=@n12.R)
  n14 = Transform(None=Constant3Vector() {constant=<rgba 0 -1 0 0>}) {transformSourceType=TRANSFORMSOURCE_Camera}
  n15 = Multiply(A=Multiply(A=@n14, B=@n7.G), B=@n12.G)
  n16 = MaterialFunctionCall(Float2=Subtract(A=TextureCoordinate()) {constB=0.5}) {materialFunction=&BreakOutFloat2Components}
  n17 = Multiply(A=Multiply(A=@n16.R, B=@n6), B=@n12.R)
  n18 = Multiply(A=Multiply(A=@n16.G, B=@n14), B=@n12.G)
  n19 = Add(A=Subtract(A=Add(A=@n17, B=@n18), B=WorldPosition().XYZ), B=@n1)
  FunctionOutput(None=Add(A=Add(A=@n13, B=@n15), B=@n19)) {outputName=World Position Offset With Pivot Support, sortPriority=2}
  FunctionOutput(None=@n19) {
   description="Takes the particles' dimensions and transforms those values into the world space positions needed to expand particles out to that number of pixels. I.e. a particle with a scale of (1,2) would appear to be 1 pixel wide by 2 pixels tall at any distance."
   outputName=World Position Offset
   sortPriority=1
  }
  n20 = FunctionInput() {
   bUsePreviewValueAsDefault=true
   description=This is an optional input. The function will correctly transform your normals.
   inputName="Normals (Optional)"
   sortPriority=10
   previewValue=<0 0 1 1>
  }
  n21 = Constant3Vector() {constant=<rgba 1 -1 -1 0>}
  FunctionOutput(None=Transform(None=Multiply(A=@n20, B=@n21)) {transformSourceType=TRANSFORMSOURCE_View, transformType=TRANSFORM_Tangent}) {description=This will output transformed tangent space normals., outputName=Normals}
 ]
}
